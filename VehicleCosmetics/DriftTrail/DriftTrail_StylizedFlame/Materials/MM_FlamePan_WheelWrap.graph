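# Translucent unlit flame material for stylized vehicle drift trails, apparently wrapping the wheels, driven by particle color and dynamic parameters.
# A panned flame texture is shaped by UV edge masks and a radial gradient.
# Bump-offset, noise-distorted masks and particle alpha erode a thickness band, whose hard Ceil cutout is subtracted from the flame, with a Stack switch for tiled bands.
# The result feeds a shared emissive function and a gradient-masked, camera-depth-faded alpha.
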Material /VehicleCosmetics/DriftTrail/DriftTrail_StylizedFlame/Materials/MM_FlamePan_WheelWrap
graph {
 properties {
  bAutomaticallySetUsageInEditor=false
  bUsedWithInstancedStaticMeshes=true
  bUsedWithMeshDeformer=true
  bUsedWithNiagaraMeshParticles=true
  bUsedWithNiagaraRibbons=true
  bUsedWithSkeletalMesh=true
  bUsedWithSplineMeshes=true
  bUsedWithVirtualHeightfieldMesh=true
  blendMode=BLEND_Translucent
  shadingModel=MSM_Unlit
  usedShadingModels=Unlit
 }
 nodes [
  n1 = TextureCoordinate()
  n2 = DynamicParameter() {paramNames [StringPower,PanSpeed,Offset,Param4]}
  n3 = AppendVector(A=@n2.Offset, B=ScalarParameter() {parameterName=V})
  n4 = Time()
  n5 = TextureSampleParameter2D(UVs=Panner(Coordinate=Add(A=@n1, B=@n3), Time=Multiply(A=@n2.PanSpeed, B=@n4)) {desc=Main Panner, speedX=2}) {parameterName=Texture_Panning, texture=/DealershipSharedLibrary/FX/Textures/TexturePackage/T_VCVFX_SolidFlame_RGB}
  MaterialExpressionConstantClamp_32 = Saturate(None=@n5.R)
  n6 = TextureCoordinate()
  n7 = ComponentMask(None=@n6) {g=true}
  n8 = Multiply(A=@n7, B=OneMinus(None=@n7))
  n9 = Constant() {r=1}
  MaterialExpressionConstantClamp_34 = Saturate(None=Multiply(A=Power(Base=@n8, Exp=@n9), B=Constant() {r=3}))
  n10 = ComponentMask(None=@n6) {r=true}
  MaterialExpressionConstantClamp_35 = Saturate(None=OneMinus(None=@n10))
  n11 = Multiply(A=@MaterialExpressionConstantClamp_34, B=@MaterialExpressionConstantClamp_35)
  MaterialExpressionConstantClamp_36 = Saturate(None=@n10)
  n12 = Multiply(A=Multiply(A=@MaterialExpressionConstantClamp_35, B=@MaterialExpressionConstantClamp_36), B=Constant() {r=10})
  n13 = ComponentMask(None=TextureCoordinate()) {r=true}
  n14 = ScalarParameter() {defaultValue=1, parameterName=Add}
  n15 = Add(A=@n5.R, B=Power(Base=Add(A=@n13, B=@n14), Exp=ScalarParameter() {defaultValue=1, parameterName=Power}))
  n16 = MaterialFunctionCall(CenterPosition=VectorParameter() {defaultValue=<rgba 0.5 0.5 0.5 0>}.RGB, Radius=Constant()) {materialFunction=&RadialGradientExponential}
  n17 = Multiply(
   A=Multiply(A=Multiply(A=@n11, B=@n12), B=Constant() {r=3})
   B=Power(Base=Subtract(A=@n15, B=@n16.RadialGradientExponential), Exp=ScalarParameter() {defaultValue=1, parameterName=erode})
  )
  n18 = Constant2Vector() {g=2, r=1}
  n19 = Frac(None=Multiply(A=TextureCoordinate(), B=@n18))
  n20 = Constant2Vector() {g=-0.45, r=-0.25}
  n21 = ComponentMask(None=Rotator(Coordinate=Add(A=@n19, B=@n20), Time=Constant() {r=0.9}) {speed=3.14}) {g=true}
  MaterialExpressionConstantClamp_38 = Saturate(None=ConstantBiasScale(None=@n21) {bias=-0.015, scale=1})
  MaterialExpressionConstantClamp_39 = Saturate(None=ConstantBiasScale(None=OneMinus(None=@n21)) {bias=-0.01, scale=1})
  n22 = Constant2Vector() {g=0.45, r=-0.25}
  n23 = ComponentMask(None=Rotator(Coordinate=Add(A=@n19, B=@n22), Time=Constant() {r=0.1}) {speed=3.14}) {g=true}
  MaterialExpressionConstantClamp_40 = Saturate(None=ConstantBiasScale(None=@n23) {bias=-0.015, scale=1})
  MaterialExpressionConstantClamp_41 = Saturate(None=ConstantBiasScale(None=OneMinus(None=@n23)) {bias=-0.01, scale=1})
  n24 = Multiply(
   A=Constant() {r=50}
   B=Multiply(
    A=Multiply(A=@MaterialExpressionConstantClamp_38, B=@MaterialExpressionConstantClamp_39)
    B=Multiply(A=@MaterialExpressionConstantClamp_40, B=@MaterialExpressionConstantClamp_41)
   )
  )
  n25 = TextureCoordinate()
  n26 = Constant() {r=2}
  n27 = Multiply(A=Frac(None=Multiply(A=@n25, B=@n26)), B=@n18)
  n28 = Constant() {r=1}
  n29 = ComponentMask(None=StaticSwitchParameter(True=Panner(Coordinate=@n27, Time=@n28) {speedY=-0.5}, False=@n25) {parameterName=Stack}) {g=true}
  n30 = Multiply(A=Multiply(A=@n29, B=OneMinus(None=@n29)), B=Constant() {r=16})
  MaterialExpressionConstantClamp_37 = Saturate(None=Subtract(A=StaticSwitchParameter(True=@n24, False=@n30) {parameterName=Stack}, B=ScalarParameter() {defaultValue=1, parameterName=Thickness}))
  MaterialExpressionVertexColor_5 = ParticleColor()
  n31 = TextureCoordinate() {coordinateIndex=1}
  n32 = Multiply(A=@n31, B=Constant() {r=2})
  n33 = ComponentMask(None=StaticSwitchParameter(True=Frac(None=@n32), False=@n31) {parameterName=Stack}) {g=true}
  n34 = OneMinus(None=@n33)
  MaterialExpressionConstantClamp_44 = Saturate(None=Multiply(A=Multiply(A=@n33, B=@n34), B=Constant() {r=8}))
  n35 = SquareRoot(None=@MaterialExpressionConstantClamp_44)
  n36 = BumpOffset(Coordinate=TextureCoordinate() {vTiling=0.001}, Height=@n35) {heightRatio=1, referencePlane=1}
  n37 = DynamicParameter() {paramNames [StringPower,PanSpeed,Offset,Param4]}
  n38 = Multiply(A=@n37.PanSpeed, B=Time())
  n39 = Add(A=TextureCoordinate() {coordinateIndex=1, uTiling=2, vTiling=2}, B=Constant() {r=-1})
  n40 = StaticSwitchParameter(True=Add(A=Frac(None=@n39), B=Constant() {r=-0.5}), False=@n39) {parameterName=Stack}
  n41 = AppendVector(A=Multiply(A=Constant() {r=0.1}, B=@MaterialExpressionConstantClamp_44), B=Multiply(A=ComponentMask(None=@n40) {g=true}, B=@n37.Offset))
  n42 = TextureSample(UVs=Add(A=Panner(Coordinate=@n36, Time=@n38), B=@n41)) {texture=/DealershipSharedLibrary/FX/Textures/Noise/T_VCVFX_TrailMask_RGB}
  n43 = TextureCoordinate() {coordinateIndex=1, vTiling=0.1}
  n44 = ComponentMask(None=@n43) {r=true}
  n45 = BumpOffset(Coordinate=AppendVector(A=Multiply(A=@n44, B=@n37.StringPower), B=ComponentMask(None=@n43) {g=true}), Height=@n35) {heightRatio=1, referencePlane=1}
  n46 = Constant() {r=0.5}
  n47 = Add(
   A=Multiply(A=ComponentMask(None=@n42.RGB) {g=true, r=true}, B=Constant() {r=0.1})
   B=Panner(Coordinate=Divide(A=@n45, B=@n46), Time=ScalarParameter() {parameterName=YOffset}) {speedY=1}
  )
  n48 = TextureSample(UVs=Add(A=Panner(Coordinate=@n47, Time=@n38) {speedX=-2, speedY=2}, B=@n41)) {texture=/DealershipSharedLibrary/FX/Textures/Flipbook/T_VCVFX_Mask_A}
  n49 = Multiply(A=@n47, B=Constant2Vector() {g=1, r=0.75})
  n50 = TextureSample(UVs=Add(A=Panner(Coordinate=@n49, Time=@n38) {speedX=-2, speedY=-2}, B=@n41)) {texture=&T_VCVFX_Mask_A}
  MaterialExpressionConstantClamp_43 = Saturate(None=ConstantBiasScale(None=Add(A=@n48.RGB, B=@n50.RGB)) {bias=-0.05, scale=5})
  n51 = Constant()
  n52 = DynamicParameter() {paramNames [StringPower,PanSpeed,Offset,Param4]}
  MaterialExpressionConstantClamp_42 = Saturate(None=Add(A=@MaterialExpressionConstantClamp_43, B=StaticSwitchParameter(True=@n51, False=@n52.Param4) {defaultValue=true, parameterName=Debug}))
  n53 = Add(A=OneMinus(None=@MaterialExpressionVertexColor_5.A), B=@MaterialExpressionConstantClamp_42)
  MaterialExpressionConstantClamp_33 = Saturate(None=Ceil(None=Subtract(A=@MaterialExpressionConstantClamp_37, B=@n53)))
  MaterialExpressionConstantClamp_31 = Saturate(None=Multiply(A=Power(Base=@MaterialExpressionConstantClamp_32, Exp=@n2.StringPower), B=Subtract(A=@n17, B=@MaterialExpressionConstantClamp_33)))
  MaterialExpressionVertexColor_4 = ParticleColor()
  n54 = TextureCoordinate()
  n55 = Multiply(
   A=Multiply(
    A=Multiply(A=@MaterialExpressionConstantClamp_31, B=@MaterialExpressionVertexColor_4.A)
    B=TextureSample() {texture=/DealershipSharedLibrary/FX/Textures/Gradient_Masks/T_VCVFX_Gradiante2_A}.RGB
   )
   B=Power(Base=ComponentMask(None=@n54) {r=true}, Exp=ScalarParameter() {defaultValue=-0.1, parameterName=alpha2})
  )
  n56 = TextureCoordinate()
  n57 = Power(Base=OneMinus(None=ComponentMask(None=@n56) {r=true}), Exp=ScalarParameter() {defaultValue=1, parameterName=Alpha})
  n58 = Constant() {r=200}
  n59 = Constant() {r=50}
  MaterialExpressionConstantClamp_30 = Saturate(None=Multiply(
    A=Multiply(A=@n55, B=@n57)
    B=MaterialFunctionCall(Fade Length=@n58, Fade Offset=@n59) {materialFunction=&CameraDepthFade}.Result
   ))
  MaterialFunctionCall(Emissive=Multiply(A=@MaterialExpressionConstantClamp_31, B=@MaterialExpressionVertexColor_4.RGB)) {materialFunction=&MF_FX_EmissiveAdjustments}
 ]
}
